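# Default Lit material for the office Eclipse screen prop that creates animated TV-style static.
# Tiled UVs feed a Panner driven by floored Time times Speed, so the noise jumps in discrete steps rather than scrolling smoothly.
# The panned UVs sample a grayscale white noise texture, which is multiplied by an inverted Fresnel to fade the static at grazing angles.
# Tiling and Speed are exposed as scalar parameters.
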Material /Game/Environments/Asteria/Props/Office/Office_EclipseScreen_A/Materials/M_OfficeScreen_Static_FNM24
graph {
 properties {bAutomaticallySetUsageInEditor=false, usedShadingModels=Default Lit}
 nodes [
  Constant3Vector()
  n1 = ScalarParameter() {defaultValue=2, parameterName=Tiling}
  n2 = Time()
  n3 = ScalarParameter() {defaultValue=8, parameterName=Speed}
  n4 = Panner(Coordinate=Multiply(A=@n1, B=TextureCoordinate()), Time=Floor(None=Multiply(A=@n2, B=@n3))) {speedX=0.189, speedY=0.344}
  Multiply(A=@n1, B=@n4)
  Multiply(
   A=TextureSample(UVs=@n4) {samplerType=SAMPLERTYPE_LinearGrayscale, texture=/Game/Global/Textures/Default/Noise/T_WhiteNoise}.RGB
   B=LinearInterpolate(Alpha=Fresnel() {baseReflectFraction=0, exponent=2}) {constA=1, constB=0}
  )
 ]
}
